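# Animation Blueprint for the WackyGuy Bass backpack cosmetic, targeting the generic bass skeleton to adjust the instrument's placement.
# The input pose passes through a single Modify Bone node that additively nudges, slightly rolls and shrinks the "offset" bone in bone space.
# The Event Graph holds only unconnected default nodes, so the adjustment is static and uses no variables.
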AnimBlueprint /SparksCosmetics/Items/Cosmetics/Backpacks/WackyGuy_Bass/WackyGuy_Bass_Backpack_AnimBP
target_skeleton=/SparksCharacterCommon/Instrument/Bass/Generic/Meshes/Generic_Bass_Skeleton
graphs [
 AnimGraph {nodes [
   n1 = "Transform (Modify) Bone - Bone: offset"!(
    ComponentPose=Local To Component(LocalPose=Input Pose())
    bAlphaBoolEnabled:bool=True
    Translation="2,-4,4"
    Rotation="0,0,2"
    Scale="0.9,0.9,0.9"
   ) -> (Pose) {
    alphaInputType=Bool
    rotationMode=BMM_Additive
    rotationSpace=BCS_BoneSpace
    scaleMode=BMM_Additive
    scaleSpace=BCS_BoneSpace
    translationMode=BMM_Additive
    translationSpace=BCS_BoneSpace
    alphaBoolBlend {blendOption=Linear}
    boneToModify {boneName=offset}
    rotation {roll=2}
    scale=<0.9 0.9 0.9>
    translation=<2 -4 4>
   }
   Output Pose(Result=Component To Local(ComponentPose=@n1))
  ]}
 EventGraph {nodes [
   TryGetPawnOwner()
   Event BlueprintUpdateAnimation()
  ]}
]
variables []
